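# Sound cue for the Force Jump power, mixing five layers: jump, wind, low end, delayed lightsaber swing and air swing.
# Each layer picks a random wave, then passes through modulation, a weapon-fire sound class override and attenuation, with wind settings shared by three layers.
# Quality Level nodes switch jump, wind and saber layers between full pools and three-wave subsets.
# Concurrency override caps the cue at three instances.
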SoundCue /PlotTwist/ForcePowers/ForceJump/Sound/ForceJump_Cue
root=Mixer_0
nodes [
 {id=Attenuation_0, children [SoundClass_0], properties {attenuationSettings=/Game/Sounds/Attenuation/Fort_ForcePowers_ForceJump_Att}}
 {id=Attenuation_1, children [SoundClass_1], properties {attenuationSettings=/Game/Sounds/Attenuation/Fort_ForcePowers_ForceJump_Wind_Att}}
 {id=Attenuation_2, children [SoundClass_2], properties {attenuationSettings=&Fort_ForcePowers_ForceJump_Wind_Att}}
 {id=Attenuation_3, children [SoundClass_3], properties {attenuationSettings=/Game/Sounds/Attenuation/Fort_ForcePowers_ForceJump_Saber_Att}}
 {id=Attenuation_4, children [SoundClass_4], properties {attenuationSettings=&Fort_ForcePowers_ForceJump_Wind_Att}}
 {id=Delay_0, children [QualityLevel_1], properties {delayMax=0.22, delayMin=0.2}}
 {id=Mixer_0, children [Attenuation_0, Attenuation_1, Attenuation_2, Attenuation_3, Attenuation_4], properties {inputVolume [1,1.5,2.5,2.5,4]}}
 {id=Modulator_0, children [QualityLevel_2], properties {pitchMax=1.1, volumeMax=1, volumeMin=1}}
 {id=Modulator_1, children [QualityLevel_0], properties {pitchMax=1.2, pitchMin=1}}
 {id=Modulator_2, children [Random_3], properties {pitchMax=1}}
 {id=Modulator_3, children [Delay_0], properties {pitchMax=1.2, volumeMax=1, volumeMin=1}}
 {id=Modulator_4, children [Random_7], properties {pitchMax=1.2, pitchMin=1}}
 {id=QualityLevel_0, children [Random_1,Random_1,Random_2,Random_2]}
 {id=QualityLevel_1, children [Random_4,Random_5,Random_5,Random_5]}
 {id=QualityLevel_2, children [Random_0,Random_0,Random_6,Random_6]}
 {
  id=Random_0
  children [WavePlayer_0, WavePlayer_1, WavePlayer_2, WavePlayer_3, WavePlayer_16, WavePlayer_17, WavePlayer_18]
  properties {weights [1,1,1,1,1,1,1]}
 }
 {
  id=Random_1
  children [WavePlayer_4, WavePlayer_5, WavePlayer_6, WavePlayer_7, WavePlayer_8, WavePlayer_9, WavePlayer_10, WavePlayer_11]
  properties {weights [1,1,1,1,1,1,1,1]}
 }
 {id=Random_2, children [WavePlayer_9, WavePlayer_10, WavePlayer_11], properties {bShouldExcludeFromBranchCulling=true, weights [1,1,1]}}
 {id=Random_3, children [WavePlayer_12, WavePlayer_13, WavePlayer_14, WavePlayer_15], properties {weights [1,1,1,1]}}
 {
  id=Random_4
  children [WavePlayer_20, WavePlayer_21, WavePlayer_22, WavePlayer_23, WavePlayer_24, WavePlayer_25, WavePlayer_26, WavePlayer_27]
  properties {weights [1,1,1,1,1,1,1,1]}
 }
 {id=Random_5, children [WavePlayer_25, WavePlayer_26, WavePlayer_27], properties {bShouldExcludeFromBranchCulling=true, weights [1,1,1]}}
 {id=Random_6, children [WavePlayer_16, WavePlayer_17, WavePlayer_18], properties {bShouldExcludeFromBranchCulling=true, weights [1,1,1]}}
 {id=Random_7, children [WavePlayer_19, WavePlayer_28, WavePlayer_29, WavePlayer_30], properties {weights [1,1,1,1]}}
 {id=SoundClass_0, children [Modulator_0], properties {soundClassOverride=&Fort_WeaponRanged_Fire_3P_PSM}}
 {id=SoundClass_1, children [Modulator_1], properties {soundClassOverride=&Fort_WeaponRanged_Fire_3P_PSM}}
 {id=SoundClass_2, children [Modulator_2], properties {soundClassOverride=&Fort_WeaponRanged_Fire_3P_PSM}}
 {id=SoundClass_3, children [Modulator_3], properties {soundClassOverride=&Fort_WeaponRanged_Fire_3P_PSM}}
 {id=SoundClass_4, children [Modulator_4], properties {soundClassOverride=&Fort_WeaponRanged_Fire_3P_PSM}}
 {id=WavePlayer_0, properties {soundWaveAssetPtr=&/Force_Jump_01}}
 {id=WavePlayer_1, properties {soundWaveAssetPtr=&/Force_Jump_02}}
 {id=WavePlayer_10, properties {soundWaveAssetPtr=&/Force_Jump_Wind_06}}
 {id=WavePlayer_11, properties {soundWaveAssetPtr=&/Force_Jump_Wind_07}}
 {id=WavePlayer_12, properties {soundWaveAssetPtr=&/Force_Jump_Lowend_01}}
 {id=WavePlayer_13, properties {soundWaveAssetPtr=&/Force_Jump_Lowend_02}}
 {id=WavePlayer_14, properties {soundWaveAssetPtr=&/Force_Jump_Lowend_03}}
 {id=WavePlayer_15, properties {soundWaveAssetPtr=&/Force_Jump_Lowend_04}}
 {id=WavePlayer_16, properties {soundWaveAssetPtr=&/Force_Jump_05}}
 {id=WavePlayer_17, properties {soundWaveAssetPtr=&/Force_Jump_06}}
 {id=WavePlayer_18, properties {soundWaveAssetPtr=&/Force_Jump_07}}
 {id=WavePlayer_19, properties {soundWaveAssetPtr=&/Force_Jump_AirSwing_01}}
 {id=WavePlayer_2, properties {soundWaveAssetPtr=&/Force_Jump_03}}
 {id=WavePlayer_20, properties {soundWaveAssetPtr=&/Force_Jump_Saber_SwingLight_01}}
 {id=WavePlayer_21, properties {soundWaveAssetPtr=&/Force_Jump_Saber_SwingLight_02}}
 {id=WavePlayer_22, properties {soundWaveAssetPtr=&/Force_Jump_Saber_SwingLight_03}}
 {id=WavePlayer_23, properties {soundWaveAssetPtr=&/Force_Jump_Saber_SwingLight_04}}
 {id=WavePlayer_24, properties {soundWaveAssetPtr=&/Force_Jump_Saber_SwingLight_05}}
 {id=WavePlayer_25, properties {soundWaveAssetPtr=&/Force_Jump_Saber_SwingLight_06}}
 {id=WavePlayer_26, properties {soundWaveAssetPtr=&/Force_Jump_Saber_SwingLight_07}}
 {id=WavePlayer_27, properties {soundWaveAssetPtr=&/Force_Jump_Saber_SwingLight_08}}
 {id=WavePlayer_28, properties {soundWaveAssetPtr=&/Force_Jump_AirSwing_02}}
 {id=WavePlayer_29, properties {soundWaveAssetPtr=&/Force_Jump_AirSwing_03}}
 {id=WavePlayer_3, properties {soundWaveAssetPtr=&/Force_Jump_04}}
 {id=WavePlayer_30, properties {soundWaveAssetPtr=&/Force_Jump_AirSwing_04}}
 {id=WavePlayer_4, properties {soundWaveAssetPtr=&/Force_Jump_Wind_08}}
 {id=WavePlayer_5, properties {soundWaveAssetPtr=&/Force_Jump_Wind_01}}
 {id=WavePlayer_6, properties {soundWaveAssetPtr=&/Force_Jump_Wind_02}}
 {id=WavePlayer_7, properties {soundWaveAssetPtr=&/Force_Jump_Wind_03}}
 {id=WavePlayer_8, properties {soundWaveAssetPtr=&/Force_Jump_Wind_04}}
 {id=WavePlayer_9, properties {soundWaveAssetPtr=&/Force_Jump_Wind_05}}
]
properties {
 bOverrideConcurrency=true
 maxDistance=5500
 soundClassObject=&Fort_Player_Foley
 volumeMultiplier=0.6
 attenuationOverrides {falloffDistance=6000, attenuationShapeExtents {x=1000}}
 concurrencyOverrides {maxCount=3}
}
